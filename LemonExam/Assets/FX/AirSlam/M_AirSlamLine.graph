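# Unlit, two-sided translucent Niagara material for Air Slam line FX, drawing streaks with a linear sine across U.
# U is floor-quantized per streak, so panned noise and flame-dissolve samples vary line by line.
# Growth-driven tip/tail masks and inverse-proportional decay shape opacity, with edge and near-camera fades.
# Particle color is hue-shifted along V, dynamic parameters drive time offset, period, dissolve and growth, and SplineThicken sets width.
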Material /LemonExam/Assets/FX/AirSlam/M_AirSlamLine
graph {
 properties {
  bAutomaticallySetUsageInEditor=false
  bUsedWithNiagaraMeshParticles=true
  bUsedWithNiagaraRibbons=true
  bUsedWithNiagaraSprites=true
  blendMode=BLEND_Translucent
  shadingModel=MSM_Unlit
  twoSided=true
  usedShadingModels=Unlit
 }
 nodes [
  n1 = MaterialFunctionCall(Float2=TextureCoordinate()) {materialFunction=&BreakOutFloat2Components}
  n2 = ScalarParameter() {group=ScaleWidth, parameterName=ScaleWidth_GrayAdjust_Add}
  n3 = Saturate(None=Multiply(A=Add(A=@n1.G, B=@n2), B=ScalarParameter() {defaultValue=2, group=ScaleWidth, parameterName=ScaleWidth_GrayAdjust_Mul}))
  n4 = ScalarParameter() {defaultValue=0.5, group=ScaleWidth, parameterName=ScaleWidth_GrayAdjust_Power}
  n5 = DynamicParameter() {paramNames [TimeRandom,P,Dissolve,Width]}
  n6 = Multiply(A=@n5.Width, B=Saturate(None=Power(Base=@n3, Exp=@n4)))
  MaterialFunctionCall(WidthBase=@n6, WidthTip=@n6, Expand U or V UV Channel=StaticBool() {value=true}) {materialFunction=&SplineThicken}
  n7 = TextureCoordinate()
  n8 = Divide(B=@n5.P)
  n9 = NamedRerouteDeclaration(None=MaterialFunctionCall(
    Value=MaterialFunctionCall(Float2=@n7) {materialFunction=&BreakOutFloat2Components}.R
    Period=@n8
    -1 to 1=StaticBool()
   ) {materialFunction=&LinearSine}.Linear Sine) {name=SineLinear, nodeColor=<rgba 1 0.0234 0 1>}
  n10 = VectorParameter() {parameterName=LineDetailMask_Refine, defaultValue=<rgba 78.459 1 0 0.5>}
  n11 = NamedRerouteUsage()
  n12 = VectorParameter() {parameterName=LineDetail_Refine2, defaultValue=<rgba 3.9357 92.252 1 0.5>}
  n13 = Add(A=Multiply(
    A=MaterialFunctionCall(Texture=@n9, RGB=@n10.RGB, A=@n10.RGB) {materialFunction=&MF_FX_Refine}.Result
    B=MaterialFunctionCall(Texture=@n11, RGB=@n12.RGB, A=@n12.RGB) {materialFunction=&MF_FX_Refine}.Result
   ))
  n14 = TextureCoordinate()
  n15 = ParticleColor()
  n16 = MaterialFunctionCall(
   Hue Shift Percentage=Multiply(
    A=MaterialFunctionCall(Float2=@n14) {materialFunction=&BreakOutFloat2Components}.G
    B=ScalarParameter() {defaultValue=0.1088, parameterName=ScaleHue}
   )
   Texture=Multiply(A=VectorParameter() {parameterName=C, defaultValue=<rgba 0.5165 1.5358 3 1>}.RGB, B=@n15.RGB)
  ) {materialFunction=&HueShift}
  MaterialFunctionCall(Emissive=Divide(A=Multiply(A=@n13, B=@n16.Result))) {materialFunction=&MF_FX_EmissiveAdjustments}
  n17 = TextureObjectParameter() {
   group=DissolveT
   parameterName=DissolveT_Texture
   samplerType=SAMPLERTYPE_LinearColor
   sortPriority=0
   texture=/Game/Items/Traps/Textures/T_Plasma_Flames_01_M
  }
  n18 = VectorParameter() {group=DissolveT, parameterName=DissolveT_UVControl, channelNames {a=VPan, b=UPan, g=VScale, r=UScale}, defaultValue=<rgba 1 0.5712 0.1 0>}
  n19 = Add(A=@n5.TimeRandom, B=Time())
  n20 = MaterialFunctionCall(Float2=TextureCoordinate()) {materialFunction=&BreakOutFloat2Components}
  n21 = Divide(A=Add(A=@n20.R, B=Multiply(A=@n8) {constB=-0.75}) {constB=0}, B=@n8)
  n22 = Multiply(A=@n18.VScale, B=DynamicParameter() {parameterIndex=1, paramNames [Growth,Param2,Param3,Param4]}.Growth)
  n23 = AppendVector(A=Multiply(A=Floor(None=@n21), B=@n8), B=Add(A=@n20.G, B=@n22))
  n24 = ScalarParameter() {defaultValue=-0.096, group=DissolveT, parameterName=DissolveT_GrayAdjust_Add}
  n25 = Multiply(
   A=Add(
    A=MaterialFunctionCall(Texture=@n17, UVControl_RGB=@n18, UVControl_A=@n18, "[Override]TimeInput"=@n19, "[Override]UVInput"=@n23) {materialFunction=&FXCommon_TextureSample}.R
    B=Add(A=@n5.Dissolve, B=@n24)
   )
   B=ScalarParameter() {defaultValue=2, group=DissolveT, parameterName=DissolveT_GrayAdjust_Mul}
  )
  n26 = Saturate(None=Power(Base=Saturate(None=@n25), Exp=ScalarParameter() {defaultValue=2.2098, group=DissolveT, parameterName=DissolveT_GrayAdjust_Power}))
  n27 = MaterialFunctionCall(Float2=TextureCoordinate()) {materialFunction=&BreakOutFloat2Components}
  n28 = DynamicParameter() {parameterIndex=1, paramNames [Growth,Param2,Param3,Param4]}
  n29 = Add(A=Add(A=@n27.G, B=@n28.Growth), B=ScalarParameter() {defaultValue=0.8057, group=GrowthTip, parameterName=GrowthTip_GrayAdjust_Add})
  n30 = ScalarParameter() {defaultValue=1, group=GrowthTip, parameterName=GrowthTip_GrayAdjust_Mul}
  n31 = Power(
   Base=Saturate(None=Multiply(A=@n29, B=@n30))
   Exp=ScalarParameter() {defaultValue=2, group=GrowthTip, parameterName=GrowthTip_GrayAdjust_Power}
  )
  n32 = Add(A=Subtract(A=OneMinus(None=@n27.G), B=@n28.Growth), B=ScalarParameter() {group=GrowthTail, parameterName=GrowthTail_GrayAdjust_Add})
  n33 = ScalarParameter() {defaultValue=4, group=GrowthTail, parameterName=GrowthTail_GrayAdjust_Mul}
  n34 = Power(
   Base=Saturate(None=Multiply(A=@n32, B=@n33))
   Exp=ScalarParameter() {defaultValue=2, group=GrowthTail, parameterName=GrowthTail_GrayAdjust_Power}
  )
  n35 = Multiply(A=Multiply(A=Saturate(None=@n31), B=Saturate(None=@n34)), B=ScalarParameter() {defaultValue=248.4207, parameterName=ThirdShape_IPDE})
  n36 = MaterialFunctionCall(Float2=TextureCoordinate()) {materialFunction=&BreakOutFloat2Components}
  n37 = VectorParameter() {
   group=TipTailFade
   parameterName=TipTailFade_EdgeMask_InvertControl
   channelNames {a="", b=Power, g=Multiply, r=Offset}
   defaultValue=<rgba 0 3 2 1>
  }
  n38 = VectorParameter() {
   group=TipTailFade
   parameterName=TipTailFade_EdgeMask_Control
   channelNames {a="", b=Power, g=Multiply, r=Offset}
   defaultValue=<rgba 0 4 1.5 1>
  }
  n39 = Multiply(
   A=MaterialFunctionCall!(Emissive=Multiply(A=@n26, B=@n35), Gradient=@n9) {materialFunction=/Game/Effects/Fort_Effects/GameplayPulginsAssets/PulseLancer/Slash/InverseProportionalDecay}.Result
   B=MaterialFunctionCall!(UorV=@n36.G, Invert_EdgeControl=@n37, NoInvert_EdgeControl=@n38) {materialFunction=/Game/Effects/Niagara/MasterMaterials/Functions/Common/FXCommon_EdgeMask_Ex}.Mask
  )
  n40 = VectorParameter() {
   group=EdgeFade
   parameterName=EdgeFade_EdgeMask_Control
   channelNames {a="", b=Power, g=Multiply, r=Offset}
   defaultValue=<rgba 0 2.5794 1 1>
  }
  n41 = MaterialFunctionCall!(UorV=@n36.R, EdgeControl=@n40) {materialFunction=/Game/Effects/Niagara/MasterMaterials/Functions/Common/FXCommon_EdgeMask_Simple}
  n42 = ScalarParameter() {defaultValue=340, parameterName=NearCameraFade}
  Multiply(
   A=Multiply(A=@n15.A, B=Multiply(A=@n39, B=@n41.Mask))
   B=Power(
    Base=MaterialFunctionCall(Distance=@n42) {materialFunction=&Near_Camera_Fade}.Result
    Exp=ScalarParameter() {defaultValue=3, parameterName=NearCameraFadePower}
   )
  )
  n43 = VectorParameter() {group=LineDetail, parameterName=LineDetail_UVControl, channelNames {a=VPan, b=UPan, g=VScale, r=UScale}, defaultValue=<rgba 3 3 0 0>}
  n44 = MaterialFunctionCall(Float2=TextureCoordinate()) {materialFunction=&BreakOutFloat2Components}
  n45 = Add(A=@n44.R, B=Multiply(A=@n8) {constB=-0.75}) {constB=0}
  n46 = Multiply(A=Floor(None=Divide(A=@n45, B=@n8)), B=@n8)
  n47 = Add(A=@n44.G, B=DynamicParameter() {parameterIndex=1, paramNames [Growth,Param2,Param3,Param4]}.Growth)
  NamedRerouteDeclaration(None=MaterialFunctionCall(
    Texture=TextureObjectParameter() {group=LineDetail, parameterName=LineDetail_Texture, sortPriority=0, texture=/Game/Global/Textures/Default/Noise/Noise}
    UVControl_RGB=@n43
    UVControl_A=@n43
    "[Override]TimeInput"=@n19
    "[Override]UVInput"=AppendVector(A=@n46, B=@n47)
   ) {materialFunction=&FXCommon_TextureSample}.R) {name=SandT, nodeColor=<rgba 0 0.2969 1 1>}
  ScalarParameter() {defaultValue=75.0654, parameterName=Width}
  ScalarParameter() {defaultValue=0.25, parameterName=P}
  Subtract(A=@n20.G, B=@n22)
 ]
}
