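# Additive unlit Niagara material for the AirSlam FX shockwave, usable on mesh, sprite and ribbon particles.
# A panning plasma detail texture drives a hue shift of the base color, which is tinted by particle color and fed through emissive adjustments.
# The alpha mask multiplies opposing Fresnel falloffs, which isolate a rim band with a softened silhouette, by a powered V-axis sine, squared depth fade and particle alpha.
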Material /LemonExam/Assets/FX/AirSlam/M_FresnelShockwave
graph {
 properties {
  bAutomaticallySetUsageInEditor=false
  bUsedWithNiagaraMeshParticles=true
  bUsedWithNiagaraRibbons=true
  bUsedWithNiagaraSprites=true
  blendMode=BLEND_Additive
  shadingModel=MSM_Unlit
  usedShadingModels=Unlit
 }
 nodes [
  n1 = TextureObjectParameter() {
   group=DetailT
   parameterName=DetailT_Texture
   samplerType=SAMPLERTYPE_LinearGrayscale
   sortPriority=0
   texture=/Game/Effects/Fort_Effects/Emotes/UltraLight/Texture/T_Plasma_D_Greyscale
  }
  n2 = VectorParameter() {group=DetailT, parameterName=DetailT_UVControl, channelNames {a=VPan, b=UPan, g=VScale, r=UScale}, defaultValue=<rgba 5 5 0 0.1>}
  n3 = Multiply(
   A=MaterialFunctionCall(Texture=@n1, UVControl_RGB=@n2, UVControl_A=@n2) {materialFunction=&FXCommon_TextureSample}.R
   B=ScalarParameter() {defaultValue=0.2325, parameterName=Det2}
  )
  n4 = VectorParameter() {parameterName=Color, defaultValue=<rgba 0.1791 0.5759 1 1>}
  n5 = ParticleColor()
  MaterialFunctionCall(Emissive=Multiply(
    A=MaterialFunctionCall(Hue Shift Percentage=@n3, Texture=@n4.RGB) {materialFunction=&HueShift}.Result
    B=@n5.RGB
   )) {materialFunction=&MF_FX_EmissiveAdjustments}
  n6 = MaterialFunctionCall(
   Value=MaterialFunctionCall(Float2=TextureCoordinate()) {materialFunction=&BreakOutFloat2Components}.G
   Period=ScalarParameter() {defaultValue=1.581}
  ) {materialFunction=&LinearSine}
  n7 = ScalarParameter() {defaultValue=4.9741, parameterName=ZEXP}
  n8 = OneMinus(None=Fresnel() {baseReflectFraction=0, exponent=1})
  n9 = ScalarParameter() {group=EdgeF, parameterName=EdgeF_GrayAdjust_Add}
  n10 = Saturate(None=Multiply(A=Add(A=@n8, B=@n9), B=ScalarParameter() {defaultValue=3.6292, group=EdgeF, parameterName=EdgeF_GrayAdjust_Mul}))
  n11 = ScalarParameter() {defaultValue=15.4704, group=EdgeF, parameterName=EdgeF_GrayAdjust_Power}
  n12 = Multiply(
   A=Fresnel(ExponentIn=ScalarParameter() {defaultValue=3.1413, parameterName=FreExp}) {baseReflectFraction=0}
   B=Saturate(None=Power(Base=@n10, Exp=@n11))
  )
  n13 = DepthFade(FadeDistance=ScalarParameter() {defaultValue=50, parameterName=DF})
  n14 = Multiply(A=@n13, B=@n13)
  Multiply(A=@n5.A, B=Multiply(A=Power(Base=@n6.Linear Sine, Exp=@n7), B=Multiply(A=@n12, B=@n14)))
 ]
}
