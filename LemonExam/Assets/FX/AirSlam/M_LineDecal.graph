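# Translucent unlit Niagara material for the AirSlam effect, drawing noisy radial lines as a fake decal on surfaces behind it.
# WorldPositionBehindTranslucency minus particle position, scaled by dynamic ScaleX, gives planar XY UVs for noise-distorted radial distance.
# That distance drives decayed sine lines and edge masks, with dynamic Offset shifting line phase and mask band.
# Lines blend two colors, add noise-masked star sparkles and hue shift, then particle color tints them and scales opacity.
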Material /LemonExam/Assets/FX/AirSlam/M_LineDecal
graph {
 properties {
  bAutomaticallySetUsageInEditor=false
  bUsedWithNiagaraMeshParticles=true
  bUsedWithNiagaraRibbons=true
  bUsedWithNiagaraSprites=true
  blendMode=BLEND_Translucent
  shadingModel=MSM_Unlit
  twoSided=true
  usedShadingModels=Unlit
 }
 nodes [
  ActorPositionWS()
  n1 = TextureObjectParameter() {
   group=Noise
   parameterName=Noise_Texture
   samplerType=SAMPLERTYPE_LinearGrayscale
   sortPriority=0
   texture=&T_SimplePerlin_Greyscale
  }
  n2 = VectorParameter() {group=Noise, parameterName=Noise_UVControl, channelNames {a=VPan, b=UPan, g=VScale, r=UScale}, defaultValue=<rgba 0.7 0.75 0 -0.02>}
  n3 = MaterialFunctionCall() {materialFunction=&WorldPositionBehindTranslucency}
  n4 = Subtract(A=@n3.Result, B=ParticlePositionWS())
  n5 = DynamicParameter() {defaultValue=<rgba 1 1 1 0>, paramNames [ScaleX,ScaleY,ScaleZ,Offset]}
  n6 = ComponentMask(None=Multiply(
    A=MaterialFunctionCall(X=@n5.ScaleX, Y=@n5.ScaleX, Z=@n5.ScaleX) {materialFunction=&MakeFloat3}.Result
    B=Constant3Vector() {constant=<rgba 100 100 100 0>}
   )) {g=true, r=true}
  n7 = Divide(A=ComponentMask(None=@n4) {g=true, r=true}, B=@n6)
  n8 = MaterialFunctionCall(Texture=@n1, UVControl_RGB=@n2, UVControl_A=@n2, "[Override]UVInput"=@n7) {materialFunction=&FXCommon_TextureSample}
  n9 = VectorParameter() {
   group=Noise
   parameterName=Noise_NoiseControl
   sortPriority=60
   channelNames {a=VStrength, b=UStrength, g=VOffset, r=UOffset}
   defaultValue=<rgba -0.5 -0.5 -0.1564 0.01>
  }
  n10 = MaterialFunctionCall!(NoiseValue=@n8.RG, Control_RGB=@n9, Control_A=@n9) {materialFunction=/Game/Effects/Niagara/MasterMaterials/Functions/Common/FXCommon_NoiseControl}
  Add(A=@n10.NoiseValue)
  n11 = Add(
   A=MaterialFunctionCall(Float2=@n10.NoiseValue) {materialFunction=&BreakOutFloat2Components}.R
   B=MaterialFunctionCall(Float2=MaterialFunctionCall(Vector or UVs=@n7) {materialFunction=&VectorToRadialValue}.Radial Coordinates) {materialFunction=&BreakOutFloat2Components}.G
  )
  n12 = Saturate(None=Multiply(A=@n11) {constB=2})
  n13 = Add(A=@n5.Offset, B=ScalarParameter() {group=MaskI, parameterName=MaskI_GrayAdjust_Add})
  Add(A=@n12, B=@n13)
  n14 = Multiply(A=Power(Base=@n8.RG, Exp=ScalarParameter() {defaultValue=2, parameterName=YZp}), B=ScalarParameter() {defaultValue=0.1088, parameterName=R})
  n15 = TextureCoordinate()
  n16 = VectorParameter() {group=Star, parameterName=Star_UVControl, channelNames {a=VPan, b=UPan, g=VScale, r=UScale}, defaultValue=<rgba 10 10 0 0>}
  n17 = ComponentMask(None=LinearInterpolate(A=@n4, B=@n3.Result, Alpha=ScalarParameter() {parameterName=DisableLocalStar})) {g=true, r=true}
  n18 = VectorParameter() {parameterName=Star_Refine, defaultValue=<rgba 5.2001 300 0.4 0.5>}
  n19 = MaterialFunctionCall(
   Texture=MaterialFunctionCall!(
    Texture=TextureObjectParameter() {group=Star, parameterName=Star_Texture, sortPriority=0, texture=/Game/Global/Textures/Default/Noise/T_TileNoise_02}
    "[Override]DDXInput"=DDX(Value=@n15)
    "[Override]DDYInput"=DDY(Value=@n15)
    AdditionalUVOffset
    UVControl_RGB=@n16
    UVControl_A=@n16
    "[Override]TimeInput"
    "[Override]UVInput"=Divide(A=@n17, B=@n6)
   ) {materialFunction=/Game/Effects/Niagara/MasterMaterials/Functions/Common/FXCommon_TextureSample_DDXDDY}.R
   RGB=@n18.RGB
   A=@n18.RGB
  ) {materialFunction=&MF_FX_Refine}
  n20 = ScalarParameter() {defaultValue=2, parameterName=YZp2}
  n21 = Saturate(None=Multiply(A=Power(Base=@n8.R, Exp=@n20), B=ScalarParameter() {defaultValue=0.1088, parameterName=R2}))
  n22 = ScalarParameter() {defaultValue=0.0436, parameterName=P}
  n23 = Subtract(A=@n11, B=Multiply(A=Multiply(A=@n22, B=@n5.Offset)) {constB=1.5})
  n24 = MaterialFunctionCall!(
   Emissive=ScalarParameter() {defaultValue=65.6538, parameterName=sssssE}
   Gradient=NamedRerouteDeclaration(None=MaterialFunctionCall(Value=@n23, Period=@n22) {materialFunction=&LinearSine}.Linear Sine) {name=SineLinear, nodeColor=<rgba 1 0.0234 0 1>}
  ) {materialFunction=/Game/Effects/Fort_Effects/GameplayPulginsAssets/PulseLancer/Slash/InverseProportionalDecay}
  n25 = Add(A=@n24.Result, B=ScalarParameter() {group=ToColor, parameterName=ToColor_GrayAdjust_Add})
  n26 = ScalarParameter() {defaultValue=1.2333, group=ToColor, parameterName=ToColor_GrayAdjust_Mul}
  n27 = Power(
   Base=Saturate(None=Multiply(A=@n25, B=@n26))
   Exp=ScalarParameter() {defaultValue=0.7888, group=ToColor, parameterName=ToColor_GrayAdjust_Power}
  )
  n28 = Add(
   A=Multiply(A=Multiply(A=@n19.Result, B=@n21), B=VectorParameter() {parameterName=StarC, defaultValue=<rgba 73.3912 98.9342 200 1>}.RGB)
   B=LinearInterpolate(
    A=VectorParameter() {parameterName=Base, defaultValue=<rgba 0.0512 0.0981 0.1948 1>}.RGB
    B=VectorParameter() {parameterName=Color, defaultValue=<rgba 0.6209 1.6967 3 1>}.RGB
    Alpha=Saturate(None=@n27)
   )
  )
  n29 = ParticleColor()
  MaterialFunctionCall(Emissive=Multiply(
    A=MaterialFunctionCall(Hue Shift Percentage=@n14, Texture=@n28) {materialFunction=&HueShift}.Result
    B=@n29.RGB
   )) {materialFunction=&MF_FX_EmissiveAdjustments}
  MaterialFunctionCall() {materialFunction=&ObjectScale}
  n30 = OneMinus(None=@n12)
  n31 = Saturate(None=Multiply(A=Add(A=@n30, B=@n5.Offset), B=ScalarParameter() {defaultValue=3.1428, group=Mask, parameterName=Mask_GrayAdjust_Mul}))
  n32 = ScalarParameter() {defaultValue=2.1808, group=Mask, parameterName=Mask_GrayAdjust_Power}
  n33 = Saturate(None=Multiply(A=Subtract(A=@n12, B=@n13), B=ScalarParameter() {defaultValue=1, group=MaskI, parameterName=MaskI_GrayAdjust_Mul}))
  n34 = ScalarParameter() {defaultValue=1, group=MaskI, parameterName=MaskI_GrayAdjust_Power}
  n35 = Multiply(A=Saturate(None=Power(Base=@n31, Exp=@n32)), B=Saturate(None=Power(Base=@n33, Exp=@n34)))
  n36 = VectorParameter() {
   group=InnerOutFade
   parameterName=InnerOutFade_EdgeMask_InvertControl
   channelNames {a="", b=Power, g=Multiply, r=Offset}
   defaultValue=<rgba 0 2 2 1>
  }
  n37 = VectorParameter() {
   group=InnerOutFade
   parameterName=InnerOutFade_EdgeMask_Control
   channelNames {a="", b=Power, g=Multiply, r=Offset}
   defaultValue=<rgba 0 5 2 1>
  }
  n38 = VectorParameter() {
   group=InnerOutFade2
   parameterName=InnerOutFade2_EdgeMask_InvertControl
   channelNames {a="", b=Power, g=Multiply, r=Offset}
   defaultValue=<rgba 0.6976 2 0.496 1>
  }
  n39 = VectorParameter() {
   group=InnerOutFade2
   parameterName=InnerOutFade2_EdgeMask_Control
   channelNames {a="", b=Power, g=Multiply, r=Offset}
   defaultValue=<rgba 0 2 2 1>
  }
  n40 = Add(
   A=Multiply(
    A=MaterialFunctionCall!(UorV=@n30, Invert_EdgeControl=@n36, NoInvert_EdgeControl=@n37) {materialFunction=/Game/Effects/Niagara/MasterMaterials/Functions/Common/FXCommon_EdgeMask_Ex}.Mask
    B=@n24.Result
   )
   B=Multiply(
    A=ScalarParameter() {defaultValue=0.6544, parameterName=sBa}
    B=MaterialFunctionCall(UorV=@n30, Invert_EdgeControl=@n38, NoInvert_EdgeControl=@n39) {materialFunction=&FXCommon_EdgeMask_Ex}.Mask
   )
  )
  Multiply(A=@n29.A, B=Saturate(None=Multiply(A=@n35, B=@n40)))
  ScalarParameter() {defaultValue=0.5787, group=Mask, parameterName=Mask_GrayAdjust_Add}
 ]
}
